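; Footstep anim notify for ButterCake NPC pawns, signaling each step from an animation.
; It casts the mesh component's owner to NPC_Pawn_ButterCake_Base and calls the parent notify.
; It then copies its FootBone variable onto the pawn and calls OnFootStep, leaving the footstep response to the pawn.
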(fn Received_Notify
  (params (MeshComp :type Skeletal%20Mesh%20Component%20Object%20Reference) (Animation :type Anim%20Sequence%20Base%20Object%20Reference) (EventReference :type Anim%20Notify%20Event%20Reference%20Structure%20%28by%20ref%29))
  (bind _returnvalue (Components|GetOwner MeshComp))
  (bind _asnpc_pawn_butter_cake_base (Utilities|Casting|CastToNPC_Pawn_ButterCake_Base _returnvalue))
  (bind _returnvalue_2 (|CallParentFunction MeshComp Animation EventReference))
  (Variables|/ButterCake/Pawns/NPC_Pawn_ButterCake_Base.NPC_Pawn_ButterCake_Base_C|SetFootBone (Variables|Default|GetFootBone) _asnpc_pawn_butter_cake_base)
  (Class|NPCPawnButterCakeBase|OnFootStep _asnpc_pawn_butter_cake_base)
  (return _returnvalue_2))
