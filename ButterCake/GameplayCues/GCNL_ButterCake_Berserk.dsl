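; Looping gameplay cue that drives the berserk material look on a ButterCake NPC's body mesh.
; On start it sets the berserk state-switch parameter and plays the EnterBerzerkCurve timeline into Berzerk_Control, repeating this when customizations finish loading.
; On removal it unbinds that event and reverses the timeline, whose finish handler sets final parameter values by direction.
; Debug strings print on entry and exit.
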(event EventBeginPlay)

(event Collision|EventActorBeginOverlap
  (params (OtherActor :type Actor%20Object%20Reference)))

(event EventTick
  (params (DeltaSeconds :type Float%20%28single-precision%29)))

(event EventOnLoopingStart(EitherVFX-preferthisversion)
  (params (MyTarget :type Actor%20Object%20Reference) (Parameters :type Gameplay%20Cue%20Parameters%20Structure%20%28by%20ref%29) (ParticleComponents :type Array%20of%20FXSystem%20Component%20Object%20References) (AudioComponents :type Array%20of%20Audio%20Component%20Object%20References))
  (bind _asnpc_pawn_butter_cake_base (Utilities|Casting|CastToNPC_Pawn_ButterCake_Base MyTarget))
  (Customization|BindEventtoOnCustomizationsLoadedEvent _asnpc_pawn_butter_cake_base (delegate Custom|OnCustoLoaded))
  (Development|PrintString _asnpc_pawn_butter_cake_base "BigBoiBerzerkMode")
  (bind _returnvalue (Customization|GetSkeletalMeshforPartType (Variables|/ButterCake/Pawns/NPC_Pawn_ButterCake_Base.NPC_Pawn_ButterCake_Base_C|GetFortCharacterParts _asnpc_pawn_butter_cake_base) "Body" :!class /Script/FortniteGame.FortCharacterPartsComponent))
  (Rendering|Material|SetScalarParameterValueonMaterials _returnvalue "StateOrderSwitchIsBerserk" 0.0)
  (bind (_direction _berzerkcurve _normalcurve) (Timeline|EnterBerzerkCurve PlayFromStart)
    (:Update
      (Rendering|Material|SetScalarParameterValueonMaterials _returnvalue "Berzerk_Control" _berzerkcurve))
    (:Finished
      (switch Utilities|FlowControl|Switch|SwitchonETimelineDirection _direction
        (:Forward
          (Rendering|Material|SetScalarParameterValueonMaterials _returnvalue "Berzerk_Control" 0.0))
        (:Backward
          (Rendering|Material|SetScalarParameterValueonMaterials _returnvalue "Berzerk_Control" 1.0)
          (Rendering|Material|SetScalarParameterValueonMaterials _returnvalue "StateOrderSwitchIsBerserk" 0.0))))))

(event Custom|OnCustoLoaded
  (params (ThisPawn :type Fort%20Pawn%20Object%20Reference))
  (Development|PrintString (stale EventOnLoopingStart(EitherVFX-preferthisversion) _asnpc_pawn_butter_cake_base) "BigBoiBerzerkMode")
  (bind _returnvalue (Customization|GetSkeletalMeshforPartType (Variables|/ButterCake/Pawns/NPC_Pawn_ButterCake_Base.NPC_Pawn_ButterCake_Base_C|GetFortCharacterParts (stale EventOnLoopingStart(EitherVFX-preferthisversion) _asnpc_pawn_butter_cake_base)) "Body" :!class /Script/FortniteGame.FortCharacterPartsComponent))
  (Rendering|Material|SetScalarParameterValueonMaterials _returnvalue "StateOrderSwitchIsBerserk" 0.0)
  (Timeline|EnterBerzerkCurve PlayFromStart))

(event EventOnRemoval(EitherVFX-preferthisversion)
  (params (MyTarget :type Actor%20Object%20Reference) (Parameters :type Gameplay%20Cue%20Parameters%20Structure%20%28by%20ref%29) (ParticleComponents :type Array%20of%20FXSystem%20Component%20Object%20References) (AudioComponents :type Array%20of%20Audio%20Component%20Object%20References) (BurstCameraShakeInstance :type Camera%20Shake%20Base%20Object%20Reference) (BurstDecalInstance :type Decal%20Actor%20Object%20Reference))
  (bind _asnpc_pawn_butter_cake_base (Utilities|Casting|CastToNPC_Pawn_ButterCake_Base MyTarget))
  (Customization|UnbindEventfromOnCustomizationsLoadedEvent _asnpc_pawn_butter_cake_base (delegate Custom|OnCustoLoaded))
  (Development|PrintString _asnpc_pawn_butter_cake_base "BigBoiBerzerkExitMode")
  (bind _returnvalue (Customization|GetSkeletalMeshforPartType (Variables|/ButterCake/Pawns/NPC_Pawn_ButterCake_Base.NPC_Pawn_ButterCake_Base_C|GetFortCharacterParts _asnpc_pawn_butter_cake_base) "Body" :!class /Script/FortniteGame.FortCharacterPartsComponent))
  (Timeline|EnterBerzerkCurve ReverseFromEnd))

(fn ConstructionScript)
