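; Looping gameplay cue that drives the swimming water VFX for ButterCake.
; On loop start a vertical object trace at the cue's location finds and stores the Fort water body actor.
; Each tick it queries that water's surface, feeding depth, plane location and normal to the first particle component while within a height range, otherwise zeroing spawn probability.
; On removal it destroys that component so the Niagara effect does not linger when ButterCake is distance culled in water.
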(event EventBeginPlay)

(event Collision|EventActorBeginOverlap
  (params (OtherActor :type Actor%20Object%20Reference)))

(event EventOnLoopingStart(EitherVFX-preferthisversion)
  (params (MyTarget :type Actor%20Object%20Reference) (Parameters :type Gameplay%20Cue%20Parameters%20Structure%20%28by%20ref%29) (ParticleComponents :type Array%20of%20FXSystem%20Component%20Object%20References) (AudioComponents :type Array%20of%20Audio%20Component%20Object%20References))
  (|CallParentFunction MyTarget Parameters ParticleComponents AudioComponents)
  (bind (_outhit _returnvalue) (Collision|LineTraceForObjects :Start (+ (Transformation|GetActorLocation) "0, 0,500.000000") :End (- (Transformation|GetActorLocation) "0, 0,600.000000") :ObjectTypes (Utilities|Array|MakeArray "ObjectTypeQuery12") :bTraceComplex false :DrawDebugType "None" :bIgnoreSelf true :TraceColor "(R=1.000000,G=0.000000,B=0.000000,A=1.000000)" :TraceHitColor "(R=0.000000,G=1.000000,B=0.000000,A=1.000000)" :DrawTime 0.25))
  (if _returnvalue
    (bind _asfort_water_body_actor (Utilities|Casting|CastToFortWaterBodyActor (.HitActor (Collision|BreakHitResult _outhit))))
    (Variables|Default|SetWater%20Body _asfort_water_body_actor)))

(event EventTick
  (params (DeltaSeconds :type Float%20%28single-precision%29))
  (bind _returnvalue (Transformation|GetActorLocation))
  (|CallParentFunction DeltaSeconds)
  (bind _outwatersurfaceinfo (WaterBody|GetWaterSurfaceInfoStruct (Variables|Default|GetWater%20Body) _returnvalue true false false))
  (if (Math|Float|InRange(Float) (- (.z (Utilities|Struct|BreakWaterSurfaceInfo _outwatersurfaceinfo)) (.z _returnvalue)) -8e+01 8e+02)
    (Effects|Components|ParticleSystem|SetFloatParameter (Utilities|Array|Get(acopy) (stale EventOnLoopingStart(EitherVFX-preferthisversion) ParticleComponents)) "User.SpawnProbability" 1.0)
    (Effects|Components|ParticleSystem|SetFloatParameter (Utilities|Array|Get(acopy) (stale EventOnLoopingStart(EitherVFX-preferthisversion) ParticleComponents)) "User.WaterDepth" (.WaterDepth (Utilities|Struct|BreakWaterSurfaceInfo _outwatersurfaceinfo)))
    (Effects|Components|ParticleSystem|SetVectorParameter (Utilities|Array|Get(acopy) (stale EventOnLoopingStart(EitherVFX-preferthisversion) ParticleComponents)) "User.WaterPlaneLocation" (Utilities|Struct|BreakWaterSurfaceInfo _outwatersurfaceinfo))
    (Effects|Components|ParticleSystem|SetVectorParameter (Utilities|Array|Get(acopy) (stale EventOnLoopingStart(EitherVFX-preferthisversion) ParticleComponents)) "User.WaterPlaneNormal" (.WaterPlaneNormal (Utilities|Struct|BreakWaterSurfaceInfo _outwatersurfaceinfo)))
    (else
      (Effects|Components|ParticleSystem|SetFloatParameter (Utilities|Array|Get(acopy) (stale EventOnLoopingStart(EitherVFX-preferthisversion) ParticleComponents)) "User.SpawnProbability" 0.0))))

(event EventOnRemoval(EitherVFX-preferthisversion)
  (params (MyTarget :type Actor%20Object%20Reference) (Parameters :type Gameplay%20Cue%20Parameters%20Structure%20%28by%20ref%29) (ParticleComponents :type Array%20of%20FXSystem%20Component%20Object%20References) (AudioComponents :type Array%20of%20Audio%20Component%20Object%20References) (BurstCameraShakeInstance :type Camera%20Shake%20Base%20Object%20Reference) (BurstDecalInstance :type Decal%20Actor%20Object%20Reference))
  (|CallParentFunction MyTarget Parameters ParticleComponents AudioComponents BurstCameraShakeInstance BurstDecalInstance)
  (Components|DestroyComponent (Utilities|Array|Get(acopy) ParticleComponents) 0) ; Ensures Niagara System is removed in cases\nwhere Buttercake is distanced culled while still in water:\nFORT-435853
)

(fn OnBurst
  (params (MyTarget :type Actor%20Object%20Reference) (Parameters :type Gameplay%20Cue%20Parameters%20Structure%20%28by%20ref%29))
  (return false))

(fn ConstructionScript
  (|CallParentFunction))
