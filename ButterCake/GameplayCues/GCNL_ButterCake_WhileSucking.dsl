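; Looping gameplay cue for ButterCake's sucking state that repeatedly shows a sound indicator at the cue's target.
; On loop start, a repeating half-second timer triggers a local-for-team footstep-type indicator with a ButterCake icon override, tracking the target for Neutral and Hostile affiliations.
; On removal, the stored timer handle is cleared, ending the repeats.
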(event EventBeginPlay)

(event Collision|EventActorBeginOverlap
  (params (OtherActor :type Actor%20Object%20Reference)))

(event EventTick
  (params (DeltaSeconds :type Float%20%28single-precision%29)))

(event EventOnLoopingStart(EitherVFX-preferthisversion)
  (params (MyTarget :type Actor%20Object%20Reference) (Parameters :type Gameplay%20Cue%20Parameters%20Structure%20%28by%20ref%29) (ParticleComponents :type Array%20of%20FXSystem%20Component%20Object%20References) (AudioComponents :type Array%20of%20Audio%20Component%20Object%20References))
  (bind _returnvalue (Utilities|Time|SetTimerbyEvent (delegate Custom|SoundIndicatorLooping) 0.5 true))
  (Variables|Default|SetSoundIndicatorTimer _returnvalue))

(event Custom|SoundIndicatorLooping
  (Fortnite|ShowSoundIndicatorLocalforTeam :WorldContextObject (stale EventOnLoopingStart(EitherVFX-preferthisversion) MyTarget) :TrackedActor (stale EventOnLoopingStart(EitherVFX-preferthisversion) MyTarget) :Location (Transformation|GetActorLocation (stale EventOnLoopingStart(EitherVFX-preferthisversion) MyTarget)) :MaxAudibleDistance 2.5e+04 :IndicatorType "FootStep" :Team (Team|GetActorTeam (stale EventOnLoopingStart(EitherVFX-preferthisversion) MyTarget)) :Affiliations (Utilities|Array|MakeArray "Neutral" "Hostile") :Instigator (stale EventOnLoopingStart(EitherVFX-preferthisversion) MyTarget) :Tint "(R=0.000000,G=0.000000,B=0.000000,A=0.000000)" :OverrideIcon "/ButterCake/Icons/AV_Buttercake_32.AV_Buttercake_32" :AttenuationMaxRangeMultiplier 1.0 :AttenuationIntensityMultiplier 1.0))

(event EventOnRemoval(EitherVFX-preferthisversion)
  (params (MyTarget :type Actor%20Object%20Reference) (Parameters :type Gameplay%20Cue%20Parameters%20Structure%20%28by%20ref%29) (ParticleComponents :type Array%20of%20FXSystem%20Component%20Object%20References) (AudioComponents :type Array%20of%20Audio%20Component%20Object%20References) (BurstCameraShakeInstance :type Camera%20Shake%20Base%20Object%20Reference) (BurstDecalInstance :type Decal%20Actor%20Object%20Reference))
  (Utilities|Time|ClearandInvalidateTimerbyHandle 0 (Variables|Default|GetSoundIndicatorTimer)))

(fn ConstructionScript)
