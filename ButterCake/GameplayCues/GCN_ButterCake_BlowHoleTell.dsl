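; Gameplay cue burst for ButterCake's blowhole tell, placing its effects and sound on the target AI pawn.
; It caches the pawn's skeletal mesh, then passes a cue parameter to each spawned Niagara component as the tell duration.
; The first audio component attaches to the Niagara component, which attaches to the blowhole leaf joint after a validity check.
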(fn OnBurst(EitherVFX-preferthisversion)
  (params (MyTarget :type Actor%20Object%20Reference) (Parameters :type Gameplay%20Cue%20Parameters%20Structure%20%28by%20ref%29) (ParticleComponents :type Array%20of%20FXSystem%20Component%20Object%20References) (AudioComponents :type Array%20of%20Audio%20Component%20Object%20References) (BurstCameraShakeInstance :type Camera%20Shake%20Base%20Object%20Reference) (BurstDecalInstance :type Decal%20Actor%20Object%20Reference))
  (Utilities|FlowControl|Sequence
    (:then_0
      (bind _asfort_aipawn (Utilities|Casting|CastToFortAIPawn MyTarget))
      (LocalVariables|Setskel_src (Variables|/Script/Engine.Character|GetMesh _asfort_aipawn)))
    (:then_1
      (for _array_element ParticleComponents
        (bind _asniagara_particle_system_component (Utilities|Casting|CastToNiagaraComponent _array_element))
        (Niagara|SetNiagaraVariable(Float) _asniagara_particle_system_component "User.BlowholeTellDuration" (Ability|GameplayCue|BreakGameplayCueParameters Parameters))
        (Transformation|AttachComponentToComponent (Utilities|Array|Get(acopy) AudioComponents) _asniagara_particle_system_component "None" "SnapToTarget" "SnapToTarget" "KeepRelative" true)
        (Utilities|IsValid 0
          (:Is%20Valid
            (Transformation|AttachComponentToComponent _asniagara_particle_system_component (LocalVariables|Getskel_src) "C_Leaf_Blowhole_Leaf2_Jnt" "SnapToTarget" "KeepWorld" "KeepRelative" true))))
      (return))))
