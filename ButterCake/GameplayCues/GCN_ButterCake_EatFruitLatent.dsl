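; Latent gameplay cue for ButterCake NPCs that drives a body material effect when one eats fruit.
; On burst it casts the target to the ButterCake pawn base, prints a debug string and gets the Body skeletal mesh.
; It raises the StateOrderSwitchIsBerserk material scalar, feeds a timeline's lured curve into Lured_Control, then clears the switch on finish.
; The timeline's second curve output is unused.
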(event EventBeginPlay
  (|CallParentFunction))

(event Collision|EventActorBeginOverlap
  (params (OtherActor :type Actor%20Object%20Reference)))

(event EventTick
  (params (DeltaSeconds :type Float%20%28single-precision%29)))

(event EventOnBurst(EitherVFX-preferthisversion)
  (params (MyTarget :type Actor%20Object%20Reference) (Parameters :type Gameplay%20Cue%20Parameters%20Structure%20%28by%20ref%29) (ParticleComponents :type Array%20of%20FXSystem%20Component%20Object%20References) (AudioComponents :type Array%20of%20Audio%20Component%20Object%20References) (BurstCameraShakeInstance :type Camera%20Shake%20Base%20Object%20Reference) (BurstDecalInstance :type Decal%20Actor%20Object%20Reference))
  (bind _asnpc_pawn_butter_cake_base (Utilities|Casting|CastToNPC_Pawn_ButterCake_Base MyTarget))
  (Development|PrintString _asnpc_pawn_butter_cake_base "BigBoiEatFruit")
  (bind _returnvalue (Customization|GetSkeletalMeshforPartType (Variables|/ButterCake/Pawns/NPC_Pawn_ButterCake_Base.NPC_Pawn_ButterCake_Base_C|GetFortCharacterParts _asnpc_pawn_butter_cake_base) "Body" :!class /Script/FortniteGame.FortCharacterPartsComponent))
  (Rendering|Material|SetScalarParameterValueonMaterials _returnvalue "StateOrderSwitchIsBerserk" 1.0)
  (bind (_direction _luredcurve _nomralcurve) (Timeline|FruitEatenCurve PlayFromStart)
    (:Update
      (Rendering|Material|SetScalarParameterValueonMaterials _returnvalue "Lured_Control" _luredcurve))
    (:Finished
      (Rendering|Material|SetScalarParameterValueonMaterials _returnvalue "StateOrderSwitchIsBerserk" 0.0))))

(fn ConstructionScript)
