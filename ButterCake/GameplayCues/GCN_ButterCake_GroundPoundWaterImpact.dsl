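; Gameplay cue notify for ButterCake's ground pound water impact, feeding local water surface data to its burst Niagara effect.
; On burst, it casts the target to a Fort AI pawn and queries that pawn's current water body at the actor's location.
; It passes the surface plane location and normal, water depth and body index as parameters to the first spawned particle component.
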(fn OnBurst(Niagara)
  (params (MyTarget :type Actor%20Object%20Reference) (Parameters :type Gameplay%20Cue%20Parameters%20Structure%20%28by%20ref%29) (ParticleComponents :type Array%20of%20Niagara%20Particle%20System%20Component%20Object%20References) (AudioComponents :type Array%20of%20Audio%20Component%20Object%20References) (BurstCameraShakeInstance :type Camera%20Shake%20Base%20Object%20Reference) (BurstDecalInstance :type Decal%20Actor%20Object%20Reference))
  (bind _output (Utilities|Array|Get(acopy) ParticleComponents))
  (bind _asfort_aipawn (Utilities|Casting|CastToFortAIPawn MyTarget))
  (bind _outwatersurfaceinfo (WaterBody|GetWaterSurfaceInfoStruct (WaterBody|GetCurrentWaterBody _asfort_aipawn) (Transformation|GetActorLocation _asfort_aipawn) true false false))
  (Effects|Components|ParticleSystem|SetVectorParameter _output "WaterPlaneLocation" (Utilities|Struct|BreakWaterSurfaceInfo _outwatersurfaceinfo))
  (Effects|Components|ParticleSystem|SetIntParameter _output "WaterBodyIndex" (.WaterBodyIdx (Utilities|Struct|BreakWaterSurfaceInfo _outwatersurfaceinfo)))
  (Effects|Components|ParticleSystem|SetFloatParameter _output "WaterDepth" (.WaterDepth (Utilities|Struct|BreakWaterSurfaceInfo _outwatersurfaceinfo)))
  (Effects|Components|ParticleSystem|SetVectorParameter _output "WaterPlaneNormal" (.WaterPlaneNormal (Utilities|Struct|BreakWaterSurfaceInfo _outwatersurfaceinfo)))
  (return))
