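; Gameplay cue notify for the ButterCake push hit on a target, playing a hit marker sound for the local viewer.
; On execute it casts the target to a Fort pawn and, if that pawn is locally viewed, spawns the sound attached to its root component.
; The parent implementation is called on the successful cast path, while a failed cast does nothing.
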(fn OnExecute
  (params (MyTarget :type Actor%20Object%20Reference) (Parameters :type Gameplay%20Cue%20Parameters%20Structure%20%28by%20ref%29))
  (bind _asfort_pawn (Utilities|Casting|CastToFortPawn MyTarget)
    (:then
      (if (Pawn|IsLocallyViewed _asfort_pawn)
        (Audio|SpawnSoundAttached "/ButterCake/Audio/HitMarker_Buttercake_Push_Cue.HitMarker_Buttercake_Push_Cue" (Variables|/Script/Engine.Actor|GetRootComponent _asfort_pawn))
        (else
          _))
      (return (|CallParentFunction MyTarget Parameters)))
    (:CastFailed
      _)))
