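; GCN_ButterCake_SneezeSpecial_Nug is a gameplay cue notify that positions burst VFX for ButterCake's special "Nug" sneeze.
; On burst, it casts the target to a Fort AI pawn and gets its skeletal mesh.
; Each spawned particle component receives that mesh as the Niagara "ButtercakeMesh" parameter and snaps to a blowhole leaf joint.
; World rotation is then zeroed so the effects point straight up regardless of the animation pose.
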(fn OnBurst(EitherVFX-preferthisversion)
  (params (MyTarget :type Actor%20Object%20Reference) (Parameters :type Gameplay%20Cue%20Parameters%20Structure%20%28by%20ref%29) (ParticleComponents :type Array%20of%20FXSystem%20Component%20Object%20References) (AudioComponents :type Array%20of%20Audio%20Component%20Object%20References) (BurstCameraShakeInstance :type Camera%20Shake%20Base%20Object%20Reference) (BurstDecalInstance :type Decal%20Actor%20Object%20Reference))
  (bind _skel_mesh_src (LocalVariables|Getskel_mesh_src))
  (bind _asfort_aipawn (Utilities|Casting|CastToFortAIPawn MyTarget))
  (LocalVariables|Setskel_mesh_src (Variables|/Script/Engine.Character|GetMesh _asfort_aipawn))
  (for _array_element ParticleComponents
    (bind _asniagara_particle_system_component (Utilities|Casting|CastToNiagaraComponent _array_element))
    (Niagara|SetNiagaraSkeletalMeshComponent _asniagara_particle_system_component "ButtercakeMesh" _skel_mesh_src)
    (Transformation|AttachComponentToComponent _array_element _skel_mesh_src "C_Leaf_Blowhole_Leaf2_Jnt" "SnapToTarget" "KeepWorld" "KeepWorld" true)
    ;[1 0.009645 0] Always have these FX point straight up in the air, regardless of animation position
    (Transformation|SetWorldRotation _array_element "0, 0, 0" false true))
  (return))
