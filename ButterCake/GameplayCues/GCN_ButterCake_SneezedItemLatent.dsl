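; Latent burst gameplay cue for ButterCake's sneezed items, adding VFX and sound to a loot pickup.
; The burst event casts the cue's SourceObject to FortPickup and selects its skeletal or static mesh component, falling back to the root component.
; If that component is valid, it attaches the sneezed-loot Niagara system and an in-air launch loop sound.
; A separate OnBurst override only prints a debug string and returns true.
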(event EventBeginPlay)

(event Collision|EventActorBeginOverlap
  (params (OtherActor :type Actor%20Object%20Reference)))

(event EventOnBurst(EitherVFX-preferthisversion)
  (params (MyTarget :type Actor%20Object%20Reference) (Parameters :type Gameplay%20Cue%20Parameters%20Structure%20%28by%20ref%29) (ParticleComponents :type Array%20of%20FXSystem%20Component%20Object%20References) (AudioComponents :type Array%20of%20Audio%20Component%20Object%20References) (BurstCameraShakeInstance :type Camera%20Shake%20Base%20Object%20Reference) (BurstDecalInstance :type Decal%20Actor%20Object%20Reference))
  (bind (_normalizedmagnitude _rawmagnitude _effectcontext _matchedtagname _originaltag _aggregatedsourcetags _aggregatedtargettags _location _normal _instigator _effectcauser _sourceobject) (Ability|GameplayCue|BreakGameplayCueParameters Parameters))
  ;[1 1 1]
  (bind _asfort_pickup (Utilities|Casting|CastToFortPickup _sourceobject))
  (Utilities|IsValid (select (Class|FortPickup|IsSkeletalMeshComponent _asfort_pickup) (select (Utilities|IsValid (Class|FortPickup|GetPickupSkeletalMeshComponent _asfort_pickup) :!class /Script/Engine.KismetSystemLibrary) (Class|FortPickup|GetPickupSkeletalMeshComponent _asfort_pickup) (Variables|/Script/Engine.Actor|GetRootComponent _asfort_pickup)) (select (Utilities|IsValid (Class|FortPickup|GetPickupStaticMeshComponent _asfort_pickup) :!class /Script/Engine.KismetSystemLibrary) (Class|FortPickup|GetPickupStaticMeshComponent _asfort_pickup) (Variables|/Script/Engine.Actor|GetRootComponent _asfort_pickup)))
    (:Is%20Valid
      (Niagara|SpawnSystemAttached "/ButterCake/VFX/NS_Buttercake_SneezedLoot_Pickup.NS_Buttercake_SneezedLoot_Pickup" (select (Class|FortPickup|IsSkeletalMeshComponent _asfort_pickup) (select (Utilities|IsValid (Class|FortPickup|GetPickupSkeletalMeshComponent _asfort_pickup) :!class /Script/Engine.KismetSystemLibrary) (Class|FortPickup|GetPickupSkeletalMeshComponent _asfort_pickup) (Variables|/Script/Engine.Actor|GetRootComponent _asfort_pickup)) (select (Utilities|IsValid (Class|FortPickup|GetPickupStaticMeshComponent _asfort_pickup) :!class /Script/Engine.KismetSystemLibrary) (Class|FortPickup|GetPickupStaticMeshComponent _asfort_pickup) (Variables|/Script/Engine.Actor|GetRootComponent _asfort_pickup))) "root" "0, 0, 0" "0, 0, 0" "SnapToTarget" true)
      (Audio|SpawnSoundAttached "/ButterCake/Audio/Buttercake_Item_Launch_In_Air_Loop_Cue.Buttercake_Item_Launch_In_Air_Loop_Cue" (select (Class|FortPickup|IsSkeletalMeshComponent _asfort_pickup) (select (Utilities|IsValid (Class|FortPickup|GetPickupSkeletalMeshComponent _asfort_pickup) :!class /Script/Engine.KismetSystemLibrary) (Class|FortPickup|GetPickupSkeletalMeshComponent _asfort_pickup) (Variables|/Script/Engine.Actor|GetRootComponent _asfort_pickup)) (select (Utilities|IsValid (Class|FortPickup|GetPickupStaticMeshComponent _asfort_pickup) :!class /Script/Engine.KismetSystemLibrary) (Class|FortPickup|GetPickupStaticMeshComponent _asfort_pickup) (Variables|/Script/Engine.Actor|GetRootComponent _asfort_pickup)))))))

(fn OnBurst
  (params (MyTarget :type Actor%20Object%20Reference) (Parameters :type Gameplay%20Cue%20Parameters%20Structure%20%28by%20ref%29))
  (Development|PrintString MyTarget "active")
  (return true))

(fn ConstructionScript)
